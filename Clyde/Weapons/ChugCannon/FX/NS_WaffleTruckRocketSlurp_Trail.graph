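# Niagara trail effect for the ChugCannon's WaffleTruck "Slurp" rocket, with a ribbon Trail and sprite Water_Splash droplets.
# Trail uses spawn groups to anchor its first burst at User.MuzzlePosition, later particles following the system, and to mask color and alpha.
# Camera distance widens the ribbon and increases positional jitter, while system timelines drive alpha.
# Droplets inherit velocity, then fall and shrink; the Minimal emitter and two alternate ribbon-material modules are disabled.
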NiagaraSystem /Clyde/Weapons/ChugCannon/FX/NS_WaffleTruckRocketSlurp_Trail
dependencies {
 usedDataInterfaces [&NiagaraDataInterfaceCurve, &NiagaraDataInterfaceCamera, &NiagaraDataInterfaceColorCurve]
 usedDynamicInputs []
 usedModules [
  &SystemState
  &Timeline
  &self:SystemSpawnScript.SystemScriptSource.SystemScriptGraph.NiagaraNodeAssignment_0.SetVariables_#1
 ]
}
emitters {
 Minimal {
  bEnabled=false
  emitterUpdateScript [
   &EmitterState()
   &SpawnBurst_Instantaneous(Use Spawn Probability=false, UseLoopCountLimit=false)
  ]
  particleSpawnScript [&InitializeParticle(
    Color Mode=::NewEnumerator0
    Lifetime=5
    Position=$User.MuzzlePosition
    Position Mode=::NewEnumerator1
    UsePositionOffset=false
    Write Material Random=true
   )]
  particleUpdateScript [&ParticleState()]
  rendererClasses [&NiagaraSpriteRendererProperties, &NiagaraMeshRendererProperties]
  renderers [
   {rendererClass=&NiagaraSpriteRendererProperties}
   {rendererClass=&NiagaraMeshRendererProperties, rendererIndex=1}
  ]
 }
 Trail {
  emitterUpdateScript [
   &EmitterState()
   &SpawnBurst_Instantaneous(Use Spawn Probability=false, UseLoopCountLimit=false)
   &SpawnBurst_Instantaneous(Spawn Group=1, Spawn Time=0.01, Use Spawn Probability=false, UseLoopCountLimit=false) {module=SpawnBurst_Instantaneous001}
   &SpawnRate(SpawnGroup=2, SpawnRate=40, Use Spawn Probability=false)
  ]
  particleSpawnScript [&InitializeParticle(
    Color Mode=::NewEnumerator0
    Lifetime=0.9
    Position=&self:ScratchDynamicInput_01(
     MaskedPosition:NiagaraPosition=$User.MuzzlePosition
     SpawnGroupMask:NiagaraInt32=0
     PasstrhoughPosition:NiagaraPosition=/Niagara/DynamicInputs/Helpers/SimulationPosition()
    )
    Position Mode=::NewEnumerator1
    Position Offset=&MaskVectorBySpawnGroup(Masked Vector=<0 0 0>, Passthrough Vector=$System.Offset)
    Ribbon Width=40
    Ribbon Width Mode=::NewEnumerator1
    UsePositionOffset=true
    Write Material Random=true
   )]
  particleUpdateScript [
   &ParticleState()
   /SetVariables(Particles.ScaleDsit:NiagaraFloat=&ClampFloat(Float=&Divide_Float(
      A=&DistanceBetweenPositions(
       A=&ReturnCameraPosition(Coordinate Space:ENiagaraCoordinateSpace=::Simulation, Camera Query:NiagaraDataInterfaceCamera=DataInterface())
       B=&SimulationPosition()
      )
      B=10000
     ))) {isSetParameters=true}
   &FN_RibbonMaterial(
    Color=&MaskLinearColorBySpawnGroup(
     Masked Linear Color=&ColorFromCurve(ColorCurve=DataInterface() {propertyValues {
        AlphaCurve {keys [{interpMode=RCIM_Cubic, tangentMode=RCTM_User, time=1, value=1}]}
        BlueCurve {keys [
          {interpMode=RCIM_Cubic, leaveTangentWeight=0.1011, tangentMode=RCTM_User, time=0.026, value=0}
          {arriveTangentWeight=0.0521, interpMode=RCIM_Cubic, leaveTangentWeight=0.1308, tangentMode=RCTM_User, time=0.1817, value=0.7}
          {arriveTangentWeight=0.2249, interpMode=RCIM_Cubic, tangentMode=RCTM_User, time=0.5724, value=0.05}
         ]}
        GreenCurve {keys [
          {interpMode=RCIM_Cubic, leaveTangentWeight=0.1011, tangentMode=RCTM_User, time=0.026, value=3.752}
          {arriveTangentWeight=0.3582, interpMode=RCIM_Cubic, leaveTangentWeight=0.8983, tangentMode=RCTM_User, time=0.1817, value=0.3142}
          {arriveTangentWeight=0.2249, interpMode=RCIM_Cubic, tangentMode=RCTM_User, time=0.5724, value=0.0224}
         ]}
        RedCurve {keys [
          {interpMode=RCIM_Cubic, leaveTangentWeight=0.1011, tangentMode=RCTM_User, time=0.026, value=5}
          {arriveTangentWeight=0.4779, interpMode=RCIM_Cubic, leaveTangentWeight=1.1986, tangentMode=RCTM_User, time=0.1817, value=0}
          {arriveTangentWeight=0.2249, interpMode=RCIM_Cubic, tangentMode=RCTM_User, time=0.5724, value=0}
         ]}
       }})
     Passthrough Linear Color=&ColorFromCurve(ColorCurve=DataInterface() {propertyValues {
        AlphaCurve {keys [<at 1 1>]}
        BlueCurve {keys [<at 0 0.7338>,<at 0.5378 1>,<at 0.978 0.1>]}
        GreenCurve {keys [<at 0 1.25>,<at 0.5378 0.4488>,<at 0.978 0.0449>]}
        RedCurve {keys [<at 0 0.3614>,<at 0.5378 0>,<at 0.978 0>]}
       }})
     Spawn Group Mask=2
    )
    Micro Distortion Amount=0.2
    Micro Pan Speed=1.35
    Opacity Pan U Speed=1.25
    Scale Alpha:NiagaraFloat=&Multiply_Float(
     A=&Multiply_Float(
      A=&FloatFromCurve(
       FloatCurve=DataInterface() {propertyValues {Curve {keys [
           {interpMode=RCIM_Cubic, leaveTangentWeight=0.6168, tangentMode=RCTM_Break, tangentWeightMode=RCTWM_WeightedBoth, time=0, value=1}
           {
            arriveTangent=-9.4255
            arriveTangentWeight=1.0887
            interpMode=RCIM_Cubic
            tangentMode=RCTM_Break
            tangentWeightMode=RCTWM_WeightedBoth
            time=1
            value=0
           }
          ]}}}
       Scale Curve=&MaskFloatBySpawnGroup(Masked Float=3, Passthrough Float=&Lerp_Float(A=1.25, B=0.7, Alpha=$Particles.NormalizedAge))
      )
      B=&FloatFromCurve(FloatCurve=DataInterface() {propertyValues {Curve {keys [<at 0.9 1>,<at 1 0>]}}}, CurveIndex=$System.Timeline.NormalizedTime)
     )
     B=&Lerp_Float(
      A=5
      B=1
      Alpha=&FloatFromCurve(
       FloatCurve=DataInterface() {propertyValues {Curve {keys [{tangentMode=RCTM_User, time=0, value=0}, {tangentMode=RCTM_User, time=1, value=1}]}}}
       CurveIndex=$System.Timeline001.NormalizedTime
      )
     )
    )
   )
   &FN_RibbonMaterial(
    Color=&MaskLinearColorBySpawnGroup(
     Masked Linear Color=&ColorFromCurve(ColorCurve=DataInterface() {propertyValues {
        AlphaCurve {keys [<at 1 1>]}
        BlueCurve {keys [<at 0.412 1.25>,<at 0.609 1.25>,<at 0.86 0>]}
        GreenCurve {keys [<at 0.412 0.561>,<at 0.609 0.1032>,<at 0.86 0.3766>]}
        RedCurve {keys [<at 0.412 0>,<at 0.609 0.9147>,<at 0.86 1.25>]}
       }})
     Passthrough Linear Color=&ColorFromCurve(ColorCurve=DataInterface() {propertyValues {
        AlphaCurve {keys [<at 1 1>]}
        BlueCurve {keys [<at 0.0612 1.25>,<at 0.9285 1.25>]}
        GreenCurve {keys [<at 0.0612 0.561>,<at 0.9285 0.1032>]}
        RedCurve {keys [<at 0.0612 0>,<at 0.9285 0.9147>]}
       }})
     Spawn Group Mask=2
    )
    Micro Distortion Amount=0.2
    Micro Pan Speed=1.35
    Opacity Pan U Speed=1.25
    Scale Alpha:NiagaraFloat=&Multiply_Float(
     A=&Multiply_Float(
      A=&FloatFromCurve(
       FloatCurve=DataInterface() {propertyValues {Curve {keys [
           {interpMode=RCIM_Cubic, tangentMode=RCTM_Break, time=0, value=1}
           {arriveTangent=-3.2997, interpMode=RCIM_Cubic, tangentMode=RCTM_Break, time=1, value=0}
          ]}}}
       Scale Curve=&MaskFloatBySpawnGroup(Masked Float=3, Passthrough Float=1.25)
      )
      B=&FloatFromCurve(FloatCurve=DataInterface() {propertyValues {Curve {keys [<at 0.9 1>,<at 1 0>]}}}, CurveIndex=$System.Timeline.NormalizedTime)
     )
     B=&Lerp_Float(
      A=5
      B=1
      Alpha=&FloatFromCurve(
       FloatCurve=DataInterface() {propertyValues {Curve {keys [{tangentMode=RCTM_User, time=0, value=0}, {tangentMode=RCTM_User, time=1, value=1}]}}}
       CurveIndex=$System.Timeline001.NormalizedTime
      )
     )
    )
   ) {enabled=false, module=FN_RibbonMaterial002}
   &FN_RibbonMaterial(
    Color=&ColorFromCurve(ColorCurve=DataInterface() {propertyValues {
       AlphaCurve {keys [<at 1 1>]}
       BlueCurve {keys [<at 0.1396 0>,<at 0.3907 1.25>,<at 0.5879 1.25>]}
       GreenCurve {keys [<at 0.1396 0.3766>,<at 0.3907 0.1032>,<at 0.5879 0.561>]}
       RedCurve {keys [<at 0.1396 1.25>,<at 0.3907 0.9147>,<at 0.5879 0>]}
      }})
    Micro Distortion Amount=0.2
    Micro Pan Speed=1.35
    Opacity Pan U Speed=1.25
    Scale Alpha:NiagaraFloat=&Multiply_Float(
     A=&Multiply_Float(
      A=&FloatFromCurve(
       FloatCurve=DataInterface() {propertyValues {Curve {keys [
           {interpMode=RCIM_Cubic, tangentMode=RCTM_Break, time=0, value=1}
           {arriveTangent=-3.2997, interpMode=RCIM_Cubic, tangentMode=RCTM_Break, time=1, value=0}
          ]}}}
       Scale Curve=&MaskFloatBySpawnGroup(Masked Float=3, Passthrough Float=1.25)
      )
      B=&FloatFromCurve(FloatCurve=DataInterface() {propertyValues {Curve {keys [<at 0.9 1>,<at 1 0>]}}}, CurveIndex=$System.Timeline.NormalizedTime)
     )
     B=&Lerp_Float(
      A=5
      B=1
      Alpha=&FloatFromCurve(
       FloatCurve=DataInterface() {propertyValues {Curve {keys [{tangentMode=RCTM_User, time=0, value=0}, {tangentMode=RCTM_User, time=1, value=1}]}}}
       CurveIndex=$System.Timeline001.NormalizedTime
      )
     )
    )
   ) {enabled=false, module=FN_RibbonMaterial001}
   &ScaleRibbonWidth(Ribbon Width Scale=&Lerp_Float(A=1, B=1.5, Alpha=$Particles.ScaleDsit))
   &ShapeLocation(
    MaskSpawnGroup=false
    Override Seed=false
    Shape Origin=$Particles.Initial.Position
    Sphere Radius=&Multiply_Float(
     A=&FloatFromCurve(
      FloatCurve=DataInterface() {propertyValues {Curve {keys [<at 0 0 RCIM_Cubic>, {arriveTangent=1.2834, interpMode=RCIM_Cubic, leaveTangent=1.2834, tangentMode=RCTM_User, time=1, value=1}]}}}
      Scale Curve=25
     )
     B=&FloatFromCurve(
      FloatCurve=DataInterface() {propertyValues {Curve {keys [<at 0 0 RCIM_Cubic>, {arriveTangent=2.431, tangentMode=RCTM_User, time=1, value=1}]}}}
      CurveIndex=$Particles.ScaleDsit
     )
    )
    UseOwnerScale=true
   )
  ]
  rendererClasses [&NiagaraRibbonRendererProperties]
  renderers [
   {rendererClass=&NiagaraRibbonRendererProperties}
   {rendererClass=&NiagaraRibbonRendererProperties, rendererIndex=1}
  ]
 }
 Water_Splash {
  emitterUpdateScript [
   &EmitterState()
   &SpawnRate(SpawnRate=15, Use Spawn Probability=false)
  ]
  particleSpawnScript [
   &InitializeParticle(
    Color Mode=::NewEnumerator0
    Lifetime=&RandomRangeFloat(Minimum=0.35, Maximum=0.5)
    Mass=&RandomRangeFloat(Minimum=0.75, Maximum=18)
    Mass Mode=::NewEnumerator1
    Position Offset=$System.Offset
    Sprite Size Mode=::NewEnumerator1
    Uniform Sprite Size Max=200
    Uniform Sprite Size Min=50
    UsePositionOffset=true
    Write Material Random=true
   )
   &InheritVelocity(
    Inherited Velocity Amount Scale=&VectorFromFloat(Value=0.5)
    Speed Limit=false
    UseSpeedThreshold=false
   )
   &ShapeLocation(
    Cylinder Height=1
    Cylinder Radius=50
    MaskSpawnGroup=false
    Override Seed=false
    Rotation Axis=&NormalizeVector(Vector To Normalize=$Engine.Owner.Velocity, Safe Normalize=true, Safe Fallback Threshold=0, Safe Fallback Vector=<0 0 0.001>)
    Rotation Mode=::NewEnumerator1
    Shape Primitive=::NewEnumerator1
    Surface Only=false
    UseOwnerScale=true
   ) {enabled=false}
   &ShapeLocation(MaskSpawnGroup=false, Override Seed=false, Sphere Radius=10, UseOwnerScale=true) {module=ShapeLocation001}
   &AddVelocity(Module.Use Falloff Distance=false, Override Seed=false, Velocity Mode=::NewEnumerator1, Velocity Speed=100)
   &AddVelocity(Velocity=&RandomRangeVector(Minimum=<-1 -1 0>, Maximum=<1 1 0>), Velocity Speed Scale=250) {module=AddVelocity001}
  ]
  particleUpdateScript [
   &ParticleState()
   &ScaleSpriteSize(Uniform Curve Sprite Scale=DataInterface() {propertyValues {Curve {keys [{tangentMode=RCTM_User, time=0, value=1}, {tangentMode=RCTM_User, time=1, value=0}]}}})
   &Drag(
    Drag=&FloatFromCurve(FloatCurve=DataInterface() {propertyValues {Curve {keys [{interpMode=RCIM_Cubic, tangentMode=RCTM_Break, time=0, value=2.5}, {interpMode=RCIM_Cubic, tangentMode=RCTM_Break, time=1, value=4}]}}})
    Use Linear Drag=true
    Use Rotational Drag=false
   )
   &GravityForce(Gravity=&Multiply_VectorByFloat(Vector=<0 0 -980>, Float=3))
   &SolveForcesAndVelocity(Clamp Velocity=false, Limit Acceleration=false)
   &ScaleSpriteSizeBySpeed(Max Scale Factor=<0.3 1>, Min Scale Factor=<0.125 0.5>)
   &ScaleColor(
    Scale Alpha=&FloatFromCurve(
     FloatCurve=DataInterface() {propertyValues {Curve {keys [
         {interpMode=RCIM_Cubic, tangentMode=RCTM_Break, time=0, value=1}
         {arriveTangent=-2.6488, interpMode=RCIM_Cubic, tangentMode=RCTM_Break, time=1, value=0}
        ]}}}
     Scale Curve=&RandomRangeFloat(Minimum=1, Maximum=1.5)
    )
    Scale RGB=&VectorFromFloat(Value=2)
    ScaleA=true
    ScaleRGB=true
   )
   &DynamicMaterialParameters(
    Index 3 Param 3=&RandomRangeFloat(Minimum=0.5, Maximum=1)
    Index 3 Param 4=&FloatFromCurve(FloatCurve=DataInterface() {propertyValues {Curve {keys [{tangentMode=RCTM_User, time=0.5, value=0}, {tangentMode=RCTM_User, time=1, value=1}]}}})
    Module.Index3Param0Write=false
    Module.Index3Param1Write=false
    Module.Index3Param2Write=true
    Module.Index3Param3Write=true
    Write Parameter Index 0=false
    Write Parameter Index 3=true
   )
  ]
  rendererClasses [&NiagaraSpriteRendererProperties]
  renderers [{rendererClass=&NiagaraSpriteRendererProperties}]
 }
}
summary {
 systemName=NS_WaffleTruckRocketSlurp_Trail
 emitters [
  {emitterName=Water_Splash, rendererClasses [&NiagaraSpriteRendererProperties]}
  {emitterName=Trail, rendererClasses [&NiagaraRibbonRendererProperties]}
  {
   bEnabled=false
   emitterName=Minimal
   rendererClasses [&NiagaraSpriteRendererProperties, &NiagaraMeshRendererProperties]
  }
 ]
 userVariables [User.MuzzlePosition {type=NiagaraPosition, description="", defaultValue {}}]
}
