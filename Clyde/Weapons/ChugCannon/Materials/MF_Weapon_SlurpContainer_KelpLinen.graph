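# Material function for the KelpLinen ChugCannon slurp container that paints animated, four-color swirling paint under glass onto a base material.
# UV-panned, distorted, bump-offset noise builds the tube color, while world-aligned noise on pre-skinned position builds the cap color.
# A warped sphere mask feeds their blend, which is hue-shifted, tinted and shaped by fresnel glass.
# The tube mask limits the BaseColor and Roughness override; the normal and radial-swirl branches never reach the output.
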MaterialFunction /Clyde/Weapons/ChugCannon/Materials/MF_Weapon_SlurpContainer_KelpLinen
graph {nodes [
  ComponentMask(None=NamedRerouteUsage()) {g=true, r=true}
  ComponentMask(None=NamedRerouteUsage()) {g=true, r=true}
  ComponentMask(None=NamedRerouteUsage()) {g=true, r=true}
  ComponentMask(None=NamedRerouteUsage()) {g=true, r=true}
  ComponentMask(None=SmoothStep(Value=NamedRerouteUsage()) {constMax=0.2, constMin=0.05}) {r=true}
  n1 = TextureSampleParameter2D() {group=Paint Swirl, parameterName=Paint.TubeMask, texture=/Clyde/Weapons/ChugCannon/Textures/T_Weapon_Slurp_Bazooka_KelpLinen_FX}
  n2 = GetMaterialAttributes(None=FunctionInput() {inputName=BaseMaterial, inputType=FunctionInput_MaterialAttributes}) {attributeGetTypes [#1]}
  n3 = LinearInterpolate(
   A=ScalarParameter() {parameterName=Paint.TubeRoughness.Min}
   B=ScalarParameter() {defaultValue=1, parameterName=Paint.TubeRoughness.Max}
   Alpha=@n2.Roughness
  )
  n4 = VectorParameter() {group=Paint Swirl, parameterName=Paint.Glass.Color, defaultValue=<rgba 0 0 0 1>}
  n5 = Multiply(A=Time(), B=ScalarParameter() {defaultValue=0.1, group=Paint Swirl, parameterName=Paint.HueShiftSpeed}) {constB=0.1}
  n6 = ScalarParameter() {group=Paint Swirl, parameterName=Paint.HueShiftOffset}
  n7 = NamedRerouteUsage()
  n8 = NamedRerouteUsage()
  n9 = NamedRerouteUsage()
  n10 = Subtract(A=@n9) {constB=0.5}
  n11 = MaterialFunctionCall(X=Subtract(A=@n9) {constB=0.5}, Y=OneMinus(None=@n10), Z=@n10) {materialFunction=&MakeFloat3}
  n12 = ScalarParameter() {defaultValue=1, parameterName=Paint.Sphere.WarpIntensity}
  n13 = SphereMask(
   A=Add(A=NamedRerouteUsage(), B=Multiply(A=@n11.Result, B=@n12))
   B=VectorParameter() {parameterName=Paint.Sphere.Pos, defaultValue=<rgba 0 0 0 1>}.RGB
   Radius=ScalarParameter() {defaultValue=200, parameterName=Paint.Sphere.Radius}
   Hardness=ScalarParameter() {defaultValue=1, parameterName=Paint.Sphere.Hardness}
  ) {hardnessPercent=0.9}
  n14 = Multiply(
   A=MaterialFunctionCall(Hue Shift Percentage=Add(A=@n5, B=@n6), Texture=LinearInterpolate(A=@n7, B=@n8, Alpha=@n13)) {materialFunction=&HueShift}.Result
   B=VectorParameter() {group=Paint Swirl, parameterName=Paint.Tint, defaultValue=<rgba 1 1 1 1>}.RGB
  )
  n15 = MaterialFunctionCall(
   Brightness=ScalarParameter() {defaultValue=1, group=Paint Swirl, parameterName=Paint.Glass.FresnelBrightness}
   Exponent=ScalarParameter() {defaultValue=1, group=Paint Swirl, parameterName=Paint.Glass.FresnelExponent}
  ) {materialFunction=&MF_Fresnel_FX}
  n16 = Transform(None=VectorParameter() {group=Paint Swirl, parameterName=Paint.Glass.FresnelDirection, defaultValue=<rgba 0 1 0 1>}.RGB) {transformSourceType=TRANSFORMSOURCE_Local}
  n17 = ScalarParameter() {defaultValue=1, group=Paint Swirl, parameterName=Paint.Glass.FresnelDirection.Exponent}
  n18 = Multiply(
   A=Abs(None=MaterialFunctionCall(Axis=@n16, Exponent=@n17) {materialFunction=&AxisAlignedFresnel}.Fresnel)
   B=ScalarParameter() {defaultValue=1, group=Paint Swirl, parameterName=Paint.Glass.FresnelDirectionBlend}
  )
  n19 = Saturate(None=Add(
    A=LinearInterpolate(A=@n15.Fresnel, Alpha=@n18)
    B=ScalarParameter() {group=Paint Swirl, parameterName=Paint.Glass.FresnelBlend, sliderMax=1, sliderMin=-1}
   ))
  n20 = Saturate(None=LinearInterpolate(A=MaterialFunctionCall() {materialFunction=&MF_Fresnel_FX}.Fresnel, Alpha=@n18))
  n21 = ScalarParameter() {parameterName=Paint.Glass.Thickness.RoundOffset}
  n22 = Saturate(None=Round(None=Add(A=@n20, B=@n21)))
  n23 = LinearInterpolate(
   A=LinearInterpolate(A=@n4.RGB, B=LinearInterpolate(A=@n4.RGB, B=@n14, Alpha=@n19) {constA=0.2}, Alpha=@n22) {constA=0.2}
   B=@n13
   Alpha=ScalarParameter() {parameterName=DebugSphere}
  )
  FunctionOutput(None=BlendMaterialAttributes(
    A=@n2.MaterialAttributes
    B=SetMaterialAttributes!(MaterialAttributes=@n2.MaterialAttributes, BaseColor=@n23, Roughness=@n3) {attributeSetTypes [#2, #1]}
    Alpha=@n1.R
   ))
  n24 = MaterialFunctionCall(Float4=NamedRerouteUsage()) {materialFunction=&BreakOutFloat4Components}
  n25 = Multiply(A=Panner(
    Coordinate=TextureCoordinate()
    Speed=MaterialFunctionCall(X=@n24.R, Y=@n24.R) {materialFunction=&MakeFloat2}.Result
   ) {speedX=0.01, speedY=0.01}) {constB=6}
  n26 = Panner(
   Coordinate=TextureCoordinate()
   Speed=MaterialFunctionCall(X=@n24.B, Y=@n24.B) {materialFunction=&MakeFloat2}.Result
  ) {speedX=0.01, speedY=0.01}
  n27 = TextureSample(UVs=BumpOffset(Coordinate=Multiply(A=@n26) {constB=5}, Height=NamedRerouteUsage()), Tex=NamedRerouteUsage()) {samplerSource=SSM_Wrap_WorldGroupSettings}
  n28 = Multiply(A=@n27.R, B=ScalarParameter() {defaultValue=0.1, group=Paint Swirl, parameterName=Paint.Distortion.Strength}) {constB=0.1}
  n29 = TextureSample(UVs=BumpOffset(Coordinate=Add(A=@n25, B=@n28), Height=NamedRerouteUsage()), Tex=NamedRerouteUsage()) {samplerSource=SSM_Wrap_WorldGroupSettings}
  n30 = MaterialFunctionCall!(In=Multiply(A=@n29.R) {constB=3}, Ramp Width) {materialFunction=/Engine/Functions/Engine_MaterialFunctions01/Math/SmoothCeil}
  LinearInterpolate(
   A=@n29.R
   B=Saturate(None=Divide(A=@n30.Result) {constB=3})
   Alpha=ScalarParameter() {defaultValue=1, parameterName=Paint.Shading.UseStepNoise}
  )
  n31 = TextureObjectParameter() {
   group=Paint Swirl
   parameterName=Paint.Distortion.Texture
   samplerType=SAMPLERTYPE_Grayscale
   texture=/Game/Effects/Fort_Effects/Textures/Fire/T_WispyFireNoise_Soft_M
  }
  NamedRerouteDeclaration(None=@n31) {name=Texture.Distortion, nodeColor=<rgba 0 0.8828 1 1>}
  n32 = DotProduct(A=PreSkinnedNormal(), B=Constant3Vector() {constant=<rgba 0 1 0 0>})
  n33 = Abs(None=@n32)
  n34 = Power(Base=Saturate(None=Multiply(A=@n33) {constB=4}))
  n35 = Saturate(None=Subtract(A=Multiply(A=@n32) {constB=-1}) {constB=0.5})
  n36 = Saturate(None=Power(Base=Multiply(A=@n35) {constB=2}) {constExponent=8})
  n37 = Constant() {r=1}
  n38 = If(A=@n32, "A > B"=Constant() {r=-1}, "A == B"=@n37, "A < B"=@n37)
  NamedRerouteDeclaration(None=VertexInterpolator(VS=MaterialFunctionCall(X=@n34, Y=@n36, Z=@n38) {materialFunction=&MakeFloat3}.Result).PS) {name=FrontBackMask, nodeColor=<rgba 0.8047 1 0 1>}
  NamedRerouteDeclaration(None=@n29.R) {name=Noise, nodeColor=<rgba 0 1 0.7656 1>}
  NamedRerouteDeclaration(None=Multiply(A=Subtract(A=@n29.R) {constB=0.5}, B=ScalarParameter() {defaultValue=0.08, parameterName=Paint.UV.Warp}) {constB=0.08}) {name=BandWarp, nodeColor=<rgba 1 0 0.3047 1>}
  n39 = NamedRerouteUsage()
  n40 = ScalarParameter() {defaultValue=32, parameterName=Paint.Cap.TextureSize}
  n41 = Multiply(
   A=VectorParameter() {group=Paint Swirl, parameterName=Paint.Cap.Speed, defaultValue=<rgba -0.05 -0.03 -0.02 0.02>}.RGB
   B=ScalarParameter() {defaultValue=1, group=Paint Swirl, parameterName=Paint.Speed.MasterMulti}
  )
  n42 = Multiply(A=@n41) {constB=0.8}
  n43 = Time()
  n44 = Add(
   A=NamedRerouteUsage()
   B=Multiply(A=Multiply(A=@n42, B=@n43), B=ScalarParameter() {defaultValue=256, group=Paint Swirl, parameterName=Paint.Cap.SpeedMulti}) {constB=256}
  )
  n45 = FunctionInput() {inputName=PreSkinnedNormalPS, sortPriority=3}
  NamedRerouteDeclaration(None=ComponentMask(None=MaterialFunctionCall(TextureObject=@n39, TextureSize=@n40, WorldPosition=@n44, World Space Normal=@n45) {materialFunction=&WorldAlignedTexture}.XYZ Texture) {r=true}) {name=WSDistortionNoise, nodeColor=<rgba 0 0.9766 1 1>}
  n46 = ScalarParameter() {defaultValue=0.25, parameterName=Paint.ShadingMin}
  n47 = NamedRerouteUsage()
  n48 = VectorParameter() {parameterName=Paint.Color.D, defaultValue=<rgba 1 0 0 1>}
  n49 = ScalarParameter() {defaultValue=0.4, parameterName=Paint.Color.D.Balance}
  n50 = NamedRerouteUsage()
  n51 = ScalarParameter() {defaultValue=32, parameterName=Paint.Cap.TextureSize}
  n52 = Multiply(A=@n51) {constB=3.5}
  n53 = Time()
  n54 = ScalarParameter() {defaultValue=256, group=Paint Swirl, parameterName=Paint.Cap.SpeedMulti}
  n55 = Add(A=NamedRerouteUsage(), B=Multiply(A=Multiply(A=@n41, B=@n53), B=Multiply(A=@n54) {constB=0.9}) {constB=256})
  n56 = NamedRerouteUsage()
  n57 = Subtract(A=@n56) {constB=0.5}
  n58 = Multiply(
   A=ScalarParameter() {defaultValue=0.1, group=Paint Swirl, parameterName=Paint.Cap.Distortion.Strength}
   B=MaterialFunctionCall(X=Subtract(A=@n56) {constB=0.5}, Y=OneMinus(None=@n57), Z=@n57) {materialFunction=&MakeFloat3}.Result
  )
  n59 = Add(A=@n55, B=@n58)
  n60 = SmoothStep(
   Min=Add(A=@n49, B=ScalarParameter() {defaultValue=0.1, parameterName=Paint.Color.Softness}) {constB=0.1}
   Max=@n49
   Value=ComponentMask(None=MaterialFunctionCall(TextureObject=@n50, TextureSize=@n52, WorldPosition=@n59, World Space Normal=@n45) {materialFunction=&WorldAlignedTexture}.XYZ Texture) {r=true}
  ) {constMax=0, constMin=0.5}
  n61 = ScalarParameter() {defaultValue=0.4, parameterName=Paint.Color.C.Balance}
  n62 = Add(A=@n61, B=ScalarParameter() {defaultValue=0.1, parameterName=Paint.Color.Softness}) {constB=0.1}
  n63 = ComponentMask(None=MaterialFunctionCall(TextureObject=NamedRerouteUsage(), TextureSize=Multiply(A=@n51) {constB=4}, WorldPosition=Add(A=@n55, B=@n58), World Space Normal=@n45) {materialFunction=&WorldAlignedTexture}.XYZ Texture) {r=true}
  n64 = ScalarParameter() {defaultValue=0.5, parameterName=Paint.Color.B.Balance}
  n65 = ScalarParameter() {defaultValue=0.1, parameterName=Paint.Color.Softness}
  n66 = Multiply(A=ScalarParameter() {defaultValue=32, parameterName=Paint.Cap.TextureSize}) {constB=2.77}
  n67 = NamedRerouteUsage()
  n68 = Multiply(
   A=Multiply(A=@n41, B=Time())
   B=Multiply(A=ScalarParameter() {defaultValue=256, group=Paint Swirl, parameterName=Paint.Cap.SpeedMulti}) {constB=1.1}
  ) {constB=256}
  n69 = MaterialFunctionCall(TextureObject=NamedRerouteUsage(), TextureSize=@n66, WorldPosition=Add(A=Add(A=@n67, B=@n68), B=@n58), World Space Normal=@n45) {materialFunction=&WorldAlignedTexture}
  n70 = LinearInterpolate(
   A=LinearInterpolate(
    A=LinearInterpolate(B=@n48.RGB, Alpha=@n60) {constA=1}
    B=VectorParameter() {parameterName=Paint.Color.C, defaultValue=<rgba 1 0.0899 0.9495 1>}.RGB
    Alpha=SmoothStep(Min=@n62, Max=@n61, Value=@n63) {constMin=0.4}
   ) {constA=1}
   B=VectorParameter() {parameterName=Paint.Color.B, defaultValue=<rgba 1 0.8242 0 1>}.RGB
   Alpha=SmoothStep(Min=@n64, Max=Add(A=@n64, B=@n65) {constB=0.1}, Value=ComponentMask(None=@n69.XYZ Texture) {r=true}) {constMin=0.4}
  )
  n71 = VectorParameter() {parameterName=Paint.Color.A, defaultValue=<rgba 0.0489 0.5433 1 1>}
  n72 = ScalarParameter() {defaultValue=0.6, parameterName=Paint.Color.A.Balance}
  n73 = NamedRerouteDeclaration(None=FunctionInput() {inputName=PreSkinnedPositionPS, sortPriority=2}) {name=PreSkinnedLocalPosWS, nodeColor=<rgba 0 1 0.8125 1>}
  n74 = Multiply(A=Multiply(A=@n41, B=Time()), B=ScalarParameter() {defaultValue=256, group=Paint Swirl, parameterName=Paint.Cap.SpeedMulti}) {constB=256}
  n75 = MaterialFunctionCall(TextureObject=NamedRerouteUsage(), TextureSize=@n66, WorldPosition=Add(A=Add(A=@n73, B=@n74), B=@n58), World Space Normal=@n45) {materialFunction=&WorldAlignedTexture}
  n76 = SmoothStep(
   Min=@n72
   Max=Add(A=@n72, B=ScalarParameter() {defaultValue=0.1, parameterName=Paint.Color.Softness}) {constB=0.1}
   Value=NamedRerouteDeclaration(None=ComponentMask(None=@n75.XYZ Texture) {r=true}) {name=WSNoise, nodeColor=<rgba 1 0 0.1172 1>}
  ) {constMax=0.5, constMin=0.4}
  NamedRerouteDeclaration(None=Multiply(A=LinearInterpolate(A=@n46, Alpha=@n47) {constA=0.25}, B=LinearInterpolate(A=@n70, B=@n71.RGB, Alpha=@n76))) {name=CapColor, nodeColor=<rgba 0 0.7187 1 1>}
  n77 = TextureObjectParameter() {
   desc=OG was Bazooka KL
   group=Paint Swirl
   parameterName=Paint.MainMask.Texture
   samplerType=SAMPLERTYPE_LinearColor
   texture=/Game/Effects/SnowSandInteraction/Textures/SSI_Perlin
  }
  NamedRerouteDeclaration(None=@n77) {name=Texture.Main, nodeColor=<rgba 1 0.5391 0 1>}
  n78 = ScalarParameter() {defaultValue=0.25, parameterName=Paint.ShadingMin}
  n79 = VectorParameter() {parameterName=Paint.Color.D, defaultValue=<rgba 1 0 0 1>}
  n80 = ScalarParameter() {defaultValue=0.4, parameterName=Paint.Color.D.Balance}
  n81 = Multiply(A=MaterialFunctionCall(UVs=TextureCoordinate(), "Rotation Angle (0-1)"=Constant() {r=0.65}) {materialFunction=&CustomRotator}.Rotated Values) {constB=4}
  n82 = Multiply(A=Time(), B=ScalarParameter() {defaultValue=0.5, parameterName=Paint.Speed}) {constB=0.2}
  n83 = NamedRerouteUsage()
  n84 = Add(A=Add(A=Add(A=@n81, B=@n82), B=ComponentMask(None=@n83) {g=true}), B=NamedRerouteUsage())
  n85 = NamedRerouteUsage()
  n86 = NamedRerouteUsage()
  n87 = SmoothStep(
   Min=Add(A=@n80, B=ScalarParameter() {defaultValue=0.1, parameterName=Paint.Color.Softness}) {constB=0.1}
   Max=@n80
   Value=TextureSample(UVs=BumpOffset(Coordinate=@n84, Height=@n85), Tex=@n86) {samplerSource=SSM_Wrap_WorldGroupSettings, samplerType=SAMPLERTYPE_LinearColor}.R
  ) {constMax=0, constMin=0.5}
  n88 = ScalarParameter() {defaultValue=0.4, parameterName=Paint.Color.C.Balance}
  n89 = Add(A=@n88, B=ScalarParameter() {defaultValue=0.1, parameterName=Paint.Color.Softness}) {constB=0.1}
  n90 = MaterialFunctionCall(UVs=TextureCoordinate(), "Rotation Angle (0-1)"=Constant() {r=0.65}) {materialFunction=&CustomRotator}
  n91 = Time()
  n92 = Multiply(A=ScalarParameter() {defaultValue=0.5, parameterName=Paint.Speed}) {constB=0.9}
  n93 = Add(
   A=Add(A=Multiply(A=@n90.Rotated Values) {constB=3.5}, B=Multiply(A=@n91, B=@n92) {constB=0.2})
   B=ComponentMask(None=NamedRerouteUsage()) {b=true}
  )
  n94 = NamedRerouteUsage()
  n95 = TextureSample(UVs=BumpOffset(Coordinate=Add(A=@n93, B=@n94), Height=@n85), Tex=@n86) {samplerSource=SSM_Wrap_WorldGroupSettings, samplerType=SAMPLERTYPE_LinearColor}
  n96 = ScalarParameter() {defaultValue=0.5, parameterName=Paint.Color.B.Balance}
  n97 = ScalarParameter() {defaultValue=0.1, parameterName=Paint.Color.Softness}
  n98 = MaterialFunctionCall(UVs=TextureCoordinate(), "Rotation Angle (0-1)"=Constant() {r=0.65}) {materialFunction=&CustomRotator}
  n99 = ScalarParameter() {defaultValue=0.5, parameterName=Paint.Speed}
  n100 = Add(A=Multiply(A=Add(A=@n98.Rotated Values) {constB=0.45}) {constB=2.77}, B=Multiply(A=Time(), B=Multiply(A=@n99) {constB=1.1}) {constB=0.2})
  n101 = ComponentMask(None=NamedRerouteUsage()) {g=true}
  n102 = NamedRerouteUsage()
  n103 = BumpOffset(Coordinate=Add(A=Add(A=@n100, B=@n101), B=NamedRerouteUsage()), Height=@n102)
  n104 = NamedRerouteUsage()
  n105 = LinearInterpolate(
   A=LinearInterpolate(
    A=LinearInterpolate(B=@n79.RGB, Alpha=@n87) {constA=1}
    B=VectorParameter() {parameterName=Paint.Color.C, defaultValue=<rgba 1 0.0899 0.9495 1>}.RGB
    Alpha=SmoothStep(Min=@n89, Max=@n88, Value=@n95.R) {constMin=0.4}
   ) {constA=1}
   B=VectorParameter() {parameterName=Paint.Color.B, defaultValue=<rgba 1 0.8242 0 1>}.RGB
   Alpha=SmoothStep(
    Min=@n96
    Max=Add(A=@n96, B=@n97) {constB=0.1}
    Value=TextureSample(UVs=@n103, Tex=@n104) {samplerSource=SSM_Wrap_WorldGroupSettings, samplerType=SAMPLERTYPE_LinearColor}.R
   ) {constMin=0.4}
  )
  n106 = VectorParameter() {parameterName=Paint.Color.A, defaultValue=<rgba 0.0489 0.5433 1 1>}
  n107 = ScalarParameter() {defaultValue=0.6, parameterName=Paint.Color.A.Balance}
  n108 = TextureCoordinate()
  n109 = Constant() {r=0.65}
  n110 = Multiply(A=Add(A=MaterialFunctionCall(UVs=@n108, "Rotation Angle (0-1)"=@n109) {materialFunction=&CustomRotator}.Rotated Values) {constB=0.33}) {constB=2.77}
  n111 = Multiply(A=Time(), B=Multiply(A=ScalarParameter() {defaultValue=0.5, parameterName=Paint.Speed}) {constB=1.1}) {constB=0.2}
  n112 = NamedRerouteUsage()
  n113 = Add(A=Add(A=Add(A=@n110, B=@n111), B=ComponentMask(None=@n112) {r=true}), B=NamedRerouteUsage())
  n114 = SmoothStep(
   Min=@n107
   Max=Add(A=@n107, B=ScalarParameter() {defaultValue=0.1, parameterName=Paint.Color.Softness}) {constB=0.1}
   Value=TextureSample(UVs=BumpOffset(Coordinate=@n113, Height=@n102), Tex=@n104) {samplerSource=SSM_Wrap_WorldGroupSettings, samplerType=SAMPLERTYPE_LinearColor}.R
  ) {constMax=0.5, constMin=0.4}
  NamedRerouteDeclaration(None=Multiply(A=LinearInterpolate(A=@n78, Alpha=@n29.R) {constA=0.25}, B=LinearInterpolate(A=@n105, B=@n106.RGB, Alpha=@n114))) {name=MainTubeColor, nodeColor=<rgba 1 0.9609 0 1>}
  n115 = TextureCoordinate()
  n116 = ScalarParameter() {defaultValue=0.8, parameterName=Paint.Distortion.DotAngle}
  n117 = VectorParameter() {parameterName=Paint.Distortion.DotTilingSpeed, defaultValue=<rgba 1 1 0.01 0.01>}
  n118 = Multiply(
   A=Panner(
    Coordinate=MaterialFunctionCall(UVs=@n115, "Rotation Angle (0-1)"=@n116) {materialFunction=&CustomRotator}.Rotated Values
    Speed=MaterialFunctionCall(X=@n117.B, Y=@n117.B) {materialFunction=&MakeFloat2}.Result
   ) {speedX=0.01, speedY=0.01}
   B=MaterialFunctionCall(X=@n117.R, Y=@n117.R) {materialFunction=&MakeFloat2}.Result
  )
  n119 = ScalarParameter() {defaultValue=0.02, parameterName=Paint.Dots.MaskDistortion}
  n120 = Add(A=@n118, B=Multiply(A=@n27.R, B=@n119) {constB=0.05})
  n121 = NamedRerouteUsage()
  n122 = TextureObjectParameter() {parameterName=Paint.Dots, texture=/Clyde/Weapons/ChugCannon/Textures/T_Weapon_KelpLinen_PaintNoise}
  n123 = TextureSample(UVs=BumpOffset(Coordinate=@n118, Height=@n121), Tex=@n122) {samplerSource=SSM_Wrap_WorldGroupSettings}
  n124 = Max(A=TextureSample(UVs=BumpOffset(Coordinate=@n120, Height=@n121), Tex=@n122) {samplerSource=SSM_Wrap_WorldGroupSettings}.R, B=@n123.R)
  n125 = TextureCoordinate()
  n126 = ScalarParameter() {defaultValue=0.8, parameterName=Paint.Distortion.DotAngle}
  n127 = VectorParameter() {parameterName=Paint.Distortion.DotTilingSpeed, defaultValue=<rgba 1 1 0.01 0.01>}
  n128 = MaterialFunctionCall(X=@n127.B, Y=@n127.B) {materialFunction=&MakeFloat2}
  n129 = Multiply(
   A=Panner(
    Coordinate=MaterialFunctionCall(UVs=@n125, "Rotation Angle (0-1)"=@n126) {materialFunction=&CustomRotator}.Rotated Values
    Speed=Multiply(A=@n128.Result) {constB=0.9}
   ) {speedX=0.01, speedY=0.01}
   B=MaterialFunctionCall(X=@n127.R, Y=@n127.R) {materialFunction=&MakeFloat2}.Result
  )
  n130 = Multiply(A=@n27.R, B=@n119) {constB=0.05}
  n131 = TextureSample(UVs=BumpOffset(Coordinate=Add(A=@n129, B=@n130), Height=NamedRerouteUsage()), Tex=@n122) {samplerSource=SSM_Wrap_WorldGroupSettings}
  n132 = Max(A=@n123.R, B=@n131.G)
  n133 = Max(A=@n123.R, B=@n131.B)
  NamedRerouteDeclaration(None=Multiply(
    A=MaterialFunctionCall(X=@n124, Y=@n132, Z=@n133) {materialFunction=&MakeFloat3}.Result
    B=ScalarParameter() {defaultValue=0.1, group=Paint Swirl, parameterName=Paint.Distortion.DotWarp}
   )) {name=DotWarp, nodeColor=<rgba 0.7812 1 0 1>}
  n134 = VectorParameter() {
   group=Paint Swirl
   parameterName=Paint.Speed
   channelNames {a=Speed Y 2, b=Speed X 2, g=Speed Y 1, r=Speed X 1}
   defaultValue=<rgba -0.05 -0.03 -0.02 0.02>
  }
  NamedRerouteDeclaration(None=Multiply(A=AppendVector(A=@n134, B=@n134), B=ScalarParameter() {defaultValue=1, group=Paint Swirl, parameterName=Paint.Speed.MasterMulti})) {name=PaintSpeedV4, nodeColor=<rgba 0.2188 0 1 1>}
  n135 = VectorParameter() {
   group=Paint Swirl
   parameterName=Paint.Cap.ScaleSpeed
   channelNames {a=Speed Y, b=Speed X, g=Scale Y, r=Scale X}
   defaultValue=<rgba 1 1 0.02 0.1>
  }
  NamedRerouteDeclaration(None=Multiply(A=MaterialFunctionCall(X=@n135.Speed X, Y=@n135.Speed X) {materialFunction=&MakeFloat2}.Result) {constB=-1}) {name=CapSpeedMulti, nodeColor=<rgba 1 0 0.3281 1>}
  NamedRerouteDeclaration(None=TextureObject() {samplerType=SAMPLERTYPE_Normal, texture=/Game/Accessories/FORT_Backpacks/Backpack_M_MED_EctoCat/FX/CloudNoise_T_N}) {name=Texture.Normal, nodeColor=<rgba 1 0 0.1406 1>}
  NamedRerouteDeclaration(None=Multiply(
    A=ScalarParameter() {group=Paint Swirl, parameterName=Paint.BumpOffset.Height}
    B=MaterialFunctionCall() {materialFunction=&MF_Fresnel_FX}.Fresnel Inverted
   )) {name=BumpHeight, nodeColor=<rgba 1 0 0.8906 1>}
  n136 = NamedRerouteUsage()
  n137 = VectorParameter() {
   group=Paint Swirl
   parameterName=Paint.Cap.ScaleSpeed
   channelNames {a=Speed Y, b=Speed X, g=Scale Y, r=Scale X}
   defaultValue=<rgba 1 1 0.02 0.1>
  }
  NamedRerouteDeclaration(None=Multiply(
    A=ComponentMask(None=@n136) {b=true}
    B=MaterialFunctionCall(X=@n137.Scale X, Y=@n137.Scale X) {materialFunction=&MakeFloat2}.Result
   )) {name=CapScale, nodeColor=<rgba 0 1 0.0391 1>}
  NamedRerouteUsage()
  NamedRerouteUsage()
  NamedRerouteUsage()
  n138 = Transform(None=PreSkinnedNormal()) {transformSourceType=TRANSFORMSOURCE_Local, transformType=TRANSFORM_Tangent}
  n139 = NamedRerouteUsage()
  n140 = MaterialFunctionCall(Float2=NamedRerouteUsage()) {materialFunction=&BreakOutFloat2Components}
  n141 = Multiply(
   A=Multiply(
    A=MaterialFunctionCall(X=@n140.R, Y=@n140.R, Z=@n140.R) {materialFunction=&MakeFloat3}.Result
    B=Time()
   )
   B=ScalarParameter() {defaultValue=1, group=Paint Swirl, parameterName=Paint.Normals.Cap.SpeedMulti}
  ) {constB=256}
  n142 = Multiply(A=TextureCoordinate(), B=ScalarParameter() {defaultValue=1, group=Paint Swirl, parameterName=Paint.Normals.Scale})
  n143 = VectorParameter() {
   group=Paint Swirl
   parameterName=Paint.Normals.Speed
   channelNames {a=Speed Y 2, b=Speed X 2, g=Speed Y 1, r=Speed X 1}
   defaultValue=<rgba -0.05 -0.03 -0.02 0.02>
  }
  n144 = AppendVector(A=@n143, B=@n143)
  n145 = ScalarParameter() {defaultValue=1, group=Paint Swirl, parameterName=Paint.Normals.SpeedMasterMulti}
  n146 = MaterialFunctionCall(Float4=NamedRerouteDeclaration(None=Multiply(A=@n144, B=@n145)) {name=PaintSpeedNormals, nodeColor=<rgba 0 0.625 1 1>}) {materialFunction=&BreakOutFloat4Components}
  n147 = Panner(
   Coordinate=@n142
   Speed=MaterialFunctionCall(X=@n146.R, Y=@n146.R) {materialFunction=&MakeFloat2}.Result
  ) {speedX=-0.05, speedY=-0.03}
  n148 = TextureSample(
   UVs=Panner(
    Coordinate=@n142
    Speed=MaterialFunctionCall(X=@n146.B, Y=@n146.B) {materialFunction=&MakeFloat2}.Result
   )
   Tex=NamedRerouteUsage()
  ) {
   samplerSource=SSM_Wrap_WorldGroupSettings
   samplerType=SAMPLERTYPE_Normal
   texture=&CloudNoise_T_N
  }
  n149 = Multiply(
   A=MaterialFunctionCall(X=@n148.R, Y=@n148.R) {materialFunction=&MakeFloat2}.Result
   B=ScalarParameter() {defaultValue=0.1, group=Paint Swirl, parameterName=Paint.Normals.DistortionStrength}
  )
  n150 = TextureSample(
   UVs=BumpOffset(Coordinate=Add(A=@n147, B=@n149), Height=ScalarParameter() {group=Paint Swirl, parameterName=Paint.BumpOffset.Height})
   Tex=NamedRerouteUsage()
  ) {
   samplerSource=SSM_Wrap_WorldGroupSettings
   samplerType=SAMPLERTYPE_Normal
   texture=&CloudNoise_T_N
  }
  n151 = LinearInterpolate(
   A=MaterialFunctionCall(
    TextureObject=NamedRerouteUsage()
    TextureSize=ScalarParameter() {defaultValue=256, group=Paint Swirl, parameterName=Paint.Normals.Cap.Scale}
    Normal=VertexInterpolator(VS=@n138).PS
    WorldPosition=Add(A=@n139, B=@n141)
   ) {materialFunction=&WorldAlignedNormal}.XYZ Texture
   B=@n150.RGB
   Alpha=@n13
  )
  n152 = OneMinus(None=ScalarParameter() {defaultValue=1, group=Paint Swirl, parameterName=Paint.Normals.Intensity})
  Reroute(None=LinearInterpolate(
    A=Constant3Vector() {constant=<rgba 0 0 1 0>}
    B=MaterialFunctionCall(Normal=@n151, Flatness=@n152) {materialFunction=&FlattenNormal}.Result
    Alpha=@n22
   )) {desc=Normals}
  n153 = VectorParameter() {group=Paint Swirl, parameterName=Paint.Cap.CenterOffset, defaultValue=<rgba 0 0 0 1>}
  n154 = TransformPosition(None=Add(A=LocalPosition() {localOrigin=InstancePreSkinning}.XYZ, B=Multiply(A=@n153.RGB) {constB=-1})) {transformType=TRANSFORMPOSSOURCE_World}
  n155 = Abs(None=Multiply(
    A=ScalarParameter() {defaultValue=32, group=Paint Swirl, parameterName=Paint.Cap.UVScale}
    B=ScalarParameter() {defaultValue=1, group=Paint Swirl, parameterName=Paint.MainMask.Scale}
   ) {constA=32})
  n156 = StaticComponentMaskParameter(None=Divide(A=VertexInterpolator(VS=@n154).PS, B=Multiply(A=@n155) {constB=-1})) {defaultB=true, defaultR=true, group=Paint Swirl, parameterName=CapAxis}
  n157 = Multiply(A=@n156, B=ScalarParameter() {defaultValue=1, group=Paint Swirl, parameterName=Paint.Distortion.Scale})
  n158 = Panner(
   Coordinate=Multiply(
    A=MaterialFunctionCall(Vector or UVs=@n157) {materialFunction=&VectorToRadialValue}.Radial Coordinates
    B=NamedRerouteUsage()
   )
   Speed=NamedRerouteUsage()
  ) {speedX=-0.05, speedY=-0.03}
  n159 = NamedRerouteUsage()
  n160 = MaterialFunctionCall(Float2=TextureCoordinate()) {materialFunction=&BreakOutFloat2Components}
  n161 = DDX(Value=@n160.R)
  n162 = DDY(Value=@n160.G)
  n163 = Add(
   A=MaterialFunctionCall(Vector or UVs=@n156) {materialFunction=&VectorToRadialValue}.Radial Coordinates
   B=Multiply(
    A=TextureSample!(UVs=@n158, Tex=@n159, "DDX(UVs)"=@n161, "DDY(UVs)"=@n162, Apply View MipBias) {mipValueMode=TMVM_Derivative}.R
    B=ScalarParameter() {defaultValue=0.1, group=Paint Swirl, parameterName=Paint.Distortion.Strength}
   ) {constB=0.1}
  )
  n164 = NamedRerouteUsage()
  n165 = MaterialFunctionCall(Float2=TextureCoordinate()) {materialFunction=&BreakOutFloat2Components}
  TextureSample(
   UVs=Panner(Coordinate=Multiply(A=@n163, B=@n164), Speed=NamedRerouteUsage()) {speedX=0.1}
   Tex=NamedRerouteUsage()
   "DDX(UVs)"=DDX(Value=@n165.R)
   "DDY(UVs)"=DDY(Value=@n165.G)
  ) {
   mipValueMode=TMVM_Derivative
   samplerSource=SSM_Wrap_WorldGroupSettings
   samplerType=SAMPLERTYPE_LinearColor
   texture=/Clyde/Athena/Apollo/Water/Textures/02_Rounded/03_WaterFoam_AccentSwirls_H
  }
 ]}
